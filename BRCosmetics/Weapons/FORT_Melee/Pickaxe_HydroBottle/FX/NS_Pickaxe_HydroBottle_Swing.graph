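# Niagara system for the HydroBottle pickaxe swing, drawing ribbon trails from weapon bones or sockets.
# Claw01 spawns up to three anim trails through two ribbon renderers, shrinking width and fading alpha with System.TrailFade and tinting by time-of-day values.
# User.SwingSectionIndex selects per-swing trail count, gated by TrailFade on the last selector entry, and per-trail Z offsets.
# Soft01, a dimmer, bluer variant with thinner trails, is disabled.
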NiagaraSystem /BRCosmetics/Weapons/FORT_Melee/Pickaxe_HydroBottle/FX/NS_Pickaxe_HydroBottle_Swing
dependencies {
 usedDataInterfaces [&NiagaraDataInterfaceArrayFloat, &NiagaraDataInterfaceCurve, &NiagaraDataInterfaceColorCurve]
 usedDynamicInputs [&ColorFromCurve, &FNTimeOfDayScalar]
 usedModules [
  &SystemState
  &self:SystemSpawnScript.SystemScriptSource.SystemScriptGraph.NiagaraNodeAssignment_0.SetVariables_#1
  &self:SystemSpawnScript.SystemScriptSource.SystemScriptGraph.NiagaraNodeAssignment_1.SetVariables_#2
  &self:SystemSpawnScript.SystemScriptSource.SystemScriptGraph.NiagaraNodeAssignment_2.SetVariables_#3
 ]
}
emitters {
 Claw01 {
  emitterUpdateScript [
   &EmitterState(Life Cycle Mode=::NewEnumerator1, Loop Behavior=::NewEnumerator1, Loop Duration=0.2, UseLoopDelay=true)
   &FN_SpawnAnimTrail_v2(
    AnimTrailWidthMode=::FromFirst
    Number of Anim Trails:NiagaraInt32=&Multiply_Int(
     A=3
     B=&MakeIntFromFloat(
      Float=&self:SectionIndexFloatSelector(
       float:NiagaraFloat=1
       float001:NiagaraFloat=1
       float002:NiagaraFloat=1
       float003:NiagaraFloat=$System.TrailFade
       SwingIndex:NiagaraInt32=$User.SwingSectionIndex
      )
      Conversion Mode=::NewEnumerator2
     )
    )
   )
  ]
  particleSpawnScript [
   /Game/Effects/Niagara/Modules/AnimTrail/FN_InitializeAnimTrail/FN_InitializeAnimTrailBoneIndecies(First Bone or Socket Index=1, Second Bone or Socket Index=1)
   &FN_InitializeAnimTrailBoneIndecies(AnimTrail Selection=1, First Bone or Socket Index=3, Second Bone or Socket Index=3) {module=FN_InitializeAnimTrailBoneIndecies001}
   &FN_InitializeAnimTrailBoneIndecies(AnimTrail Selection=2, First Bone or Socket Index=5, Second Bone or Socket Index=5) {module=FN_InitializeAnimTrailBoneIndecies002}
   /Game/Effects/Niagara/Modules/AnimTrail/FN_InitializeAnimTrail/FN_InitializeAnimTrail(
    AnimTrailWidth=3
    Color=&MultiplyLinearColor(
     Linear Color=<rgba 0.3351 0.3351 0.3351 1>
     Scale Factor=&MakeVector4_FromLinearColor(LinearColor=$System.TODColor)
    )
    Lifetime:NiagaraFloat=$System.Lifetime
    Position Offset:Vector3f=&MakeVector(
     X=0
     Y=0
     Z=&self:SectionIndexFloatSelector(
      float:NiagaraFloat=0
      float001:NiagaraFloat=&SelectFloatFromArray(Direct Array Index=$Particles.SpawnGroup, Float Selection Array=DataInterface() {propertyValues {FloatData [0,0,0]}})
      float002:NiagaraFloat=&SelectFloatFromArray(Direct Array Index=$Particles.SpawnGroup, Float Selection Array=DataInterface() {propertyValues {FloatData [0,4,-4]}})
      float003:NiagaraFloat=&SelectFloatFromArray(Direct Array Index=$Particles.SpawnGroup, Float Selection Array=DataInterface() {propertyValues {FloatData [0,4,-4]}})
      SwingIndex:NiagaraInt32=$User.SwingSectionIndex
     )
    )
   )
  ]
  particleUpdateScript [
   &FN_UpdateAnimTrail_v2()
   &ScaleRibbonWidth(Ribbon Width Scale=&FloatFromCurve(FloatCurve=DataInterface() {propertyValues {Curve {keys [
         {arriveTangent=-1.2579, interpMode=RCIM_Cubic, time=0, value=1}
         {arriveTangent=-0.9664, interpMode=RCIM_Cubic, leaveTangent=-0.9664, tangentMode=RCTM_User, time=1, value=0}
        ]}}}))
   &ScaleColor(
    Scale Alpha=&Multiply_Float(
     A=&FloatFromCurve(FloatCurve=DataInterface() {propertyValues {Curve {keys [<at 0 0.25>,<at 1 0>]}}}, Scale Curve=$System.TrailFade)
     B=4
    )
    Scale RGB=&MakeVectorFromLinearColorRGB(Color=<rgba 1 1 1 1>)
    ScaleA=true
    ScaleRGB=false
   )
   &ScaleColor(Scale RGB=&VectorFromFloat(Value=$System.TODFloat), ScaleA=false, ScaleRGB=true) {module=ScaleColor001}
  ]
  rendererClasses [&NiagaraRibbonRendererProperties]
  renderers [
   {rendererClass=&NiagaraRibbonRendererProperties}
   {rendererClass=&NiagaraRibbonRendererProperties, rendererIndex=1}
  ]
 }
 Soft01 {
  bEnabled=false
  emitterUpdateScript [
   &EmitterState(Life Cycle Mode=::NewEnumerator1, Loop Behavior=::NewEnumerator1, Loop Duration=0.2, UseLoopDelay=true)
   &FN_SpawnAnimTrail_v2(Number of Anim Trails:NiagaraInt32=&Multiply_Int(
     A=3
     B=&MakeIntFromFloat(
      Float=&self:SectionIndexFloatSelector(
       float:NiagaraFloat=1
       float001:NiagaraFloat=1
       float002:NiagaraFloat=1
       float003:NiagaraFloat=$System.TrailFade
       SwingIndex:NiagaraInt32=$User.SwingSectionIndex
      )
      Conversion Mode=::NewEnumerator2
     )
    ))
  ]
  particleSpawnScript [
   &FN_InitializeAnimTrailBoneIndecies(Second Bone or Socket Index=1)
   &FN_InitializeAnimTrailBoneIndecies(AnimTrail Selection=1, First Bone or Socket Index=2, Second Bone or Socket Index=3) {module=FN_InitializeAnimTrailBoneIndecies001}
   &FN_InitializeAnimTrailBoneIndecies(AnimTrail Selection=2, First Bone or Socket Index=4, Second Bone or Socket Index=5) {module=FN_InitializeAnimTrailBoneIndecies002}
   &FN_InitializeAnimTrail(
    AnimTrailWidth=1
    Color=&MultiplyLinearColor(
     Linear Color=<rgba 0.2286 0.2572 0.3351 1>
     Scale Factor=&MakeVector4_FromLinearColor(LinearColor=$System.TODColor)
    )
    Lifetime:NiagaraFloat=$System.Lifetime
    Position Offset:Vector3f=&MakeVector(
     X=0
     Y=0
     Z=&self:SectionIndexFloatSelector(
      float:NiagaraFloat=0
      float001:NiagaraFloat=&SelectFloatFromArray(Direct Array Index=$Particles.SpawnGroup, Float Selection Array=DataInterface() {propertyValues {FloatData [0,0,0]}})
      float002:NiagaraFloat=&SelectFloatFromArray(Direct Array Index=$Particles.SpawnGroup, Float Selection Array=DataInterface() {propertyValues {FloatData [0,4,-4]}})
      float003:NiagaraFloat=&SelectFloatFromArray(Direct Array Index=$Particles.SpawnGroup, Float Selection Array=DataInterface() {propertyValues {FloatData [0,4,-4]}})
      SwingIndex:NiagaraInt32=$User.SwingSectionIndex
     )
    )
   ) {module=FN_InitializeAnimTrail001}
  ]
  particleUpdateScript [
   &FN_UpdateAnimTrail_v2()
   &ScaleRibbonWidth(Ribbon Width Scale=&FloatFromCurve(FloatCurve=DataInterface() {propertyValues {Curve {keys [
         {arriveTangent=-1.2579, interpMode=RCIM_Cubic, time=0, value=1}
         {arriveTangent=-0.9664, interpMode=RCIM_Cubic, leaveTangent=-0.9664, tangentMode=RCTM_User, time=1, value=0}
        ]}}}))
   &ScaleColor(
    Linear Color Curve=DataInterface() {propertyValues {
      AlphaCurve {keys [<at 0 1 RCIM_Cubic>,<at 1 0 RCIM_Cubic>]}
      BlueCurve {keys [<at 0.1085 1>,<at 0.2257 0.5>,<at 0.5683 0.1146 RCIM_Cubic>]}
      GreenCurve {keys [<at 0.1085 1>,<at 0.2257 0.5>,<at 0.5683 0.1146 RCIM_Cubic>]}
      RedCurve {keys [<at 0.1085 1>,<at 0.2257 0.5>,<at 0.5683 0.1146 RCIM_Cubic>]}
     }}
    Scale Mode=::NewEnumerator2
   )
   &ScaleColor(
    Scale Alpha=&Multiply_Float(A=$System.TrailFade, B=0.2)
    Scale RGB=&VectorFromFloat(Value=$System.TODFloat)
    ScaleA=true
    ScaleRGB=false
   ) {module=ScaleColor001}
  ]
  rendererClasses [&NiagaraRibbonRendererProperties]
  renderers [{rendererClass=&NiagaraRibbonRendererProperties}]
 }
}
summary {
 systemName=NS_Pickaxe_HydroBottle_Swing
 emitters [
  {emitterName=Claw01, rendererClasses [&NiagaraRibbonRendererProperties]}
  {bEnabled=false, emitterName=Soft01, rendererClasses [&NiagaraRibbonRendererProperties]}
 ]
 userVariables [
  User.RibbonIDOffset {type=NiagaraInt32}
  User.SkeletalMesh {
   type=NiagaraDataInterfaceSkeletalMesh
   defaultValue {
    struct=NiagaraExt_VariableValue_DataInterface
    value {dataInterface=&self:NiagaraDataInterfaceSkeletalMesh_0}
   }
  }
  User.SwingSectionIndex {type=NiagaraInt32}
 ]
}
