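# Material for the HydroBottle pickaxe that draws an animated sci-fi text screen, built on the 2023 character parent PreFX/PostFX functions.
# Scanlines blend two background colors, under tinted text from a tileable texture that scrolls in whole-line steps on sine-modulated time.
# Bounding-box and bottom-line masks, plus a stepped horizontal edge, shape the text like lines being typed.
# The result replaces parent emissive through a tech art mask channel, between PreFX and PostFX.
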Material /BRCosmetics/Weapons/FORT_Melee/Pickaxe_HydroBottle/Materials/M_HydroBottle_Pickaxe
graph {
 properties {
  bHasPixelAnimation=true
  bUseMaterialAttributes=true
  bUsedWithMeshDeformer=true
  bUsedWithSkeletalMesh=true
  pixelDepthOffsetMode=PDOM_Legacy
  shadingModel=MSM_FromMaterialExpression
  usedShadingModels=Default Lit | Subsurface | Clear Coat | Subsurface Profile | Cloth
 }
 nodes [
  n1 = VectorParameter() {group=Scifi_Screen, parameterName=BackgroundOrange, defaultValue=<rgba 0.8388 0.3596 0 1>}
  n2 = VectorParameter() {group=Scifi_Screen, parameterName=BackGroundYellow, defaultValue=<rgba 0.8388 0.5354 0.0994 1>}
  n3 = TextureSampleParameter2D(UVs=Multiply(A=TextureCoordinate(), B=ScalarParameter() {defaultValue=3, group=Scifi_Screen, parameterName=ScanLineTextureTilling})) {group=Scifi_Screen, parameterName=ScanLine, texture=/BRCosmetics/Weapons/FORT_Melee/Pickaxe_HydroBottle/Textures/T_ScanLines}
  n4 = VectorParameter() {
   group=Scifi_Screen
   parameterName=TextTexture_Center
   channelNames {a=ScaleY, b=ScaleX, g=CenterY, r=CenterX}
   defaultValue=<rgba 0.5 0.5 1 1>
  }
  MaterialExpressionAppendVector_19 = AppendVector(A=@n4.CenterX, B=@n4.CenterX)
  MaterialExpressionAppendVector_20 = AppendVector(A=@n4.ScaleX, B=@n4.ScaleX)
  MaterialExpressionTextureCoordinate_1 = TextureCoordinate()
  n5 = Custom!(UVPoint=@MaterialExpressionAppendVector_19, UVScale=@MaterialExpressionAppendVector_20, OriginalUV=@MaterialExpressionTextureCoordinate_1) {
   code="float2 DistUV =OriginalUV-UVPoint;\r\nDistUV = DistUV/UVScale + float2(0.5,0.5);\r\n\r\nreturn DistUV;"
   description=CopyToPointUV
   outputType=CMOT_Float2
   inputs [
    {
     inputName=UVPoint
     input {expression=&self:MaterialExpressionAppendVector_19}
    }
    {
     inputName=UVScale
     input {expression=&self:MaterialExpressionAppendVector_20}
    }
    {
     inputName=OriginalUV
     input {expression=&self:MaterialExpressionTextureCoordinate_1}
    }
   ]
  }
  n6 = AppendVector(A=Constant(), B=Divide(B=ScalarParameter() {defaultValue=10, group=Scifi_Screen, parameterName=LinesNumber}))
  n7 = Multiply(A=Time(), B=ScalarParameter() {defaultValue=1, group=Scifi_Screen, parameterName=TextSpeed})
  n8 = Divide(A=@n7, B=MaterialFunctionCall() {materialFunction=&Pi}.Result)
  n9 = Add(A=Multiply(A=Sine(None=@n8)) {constB=0.5}) {constB=0.5}
  n10 = Add(A=Add(A=@n7, B=Multiply(B=@n9) {constA=1}), B=ScalarParameter() {parameterName=TimeOffset01})
  n11 = Add(
   A=NamedRerouteDeclaration(None=@n5) {name=AdjustUV, nodeColor=<rgba 0 1 0.5313 1>}
   B=Multiply(A=Multiply(A=@n6) {constB=-1}, B=Floor(None=@n10))
  )
  n12 = TextureSampleParameter2D(UVs=@n11) {
   group=Scifi_Screen
   parameterName=Text_Blur
   samplerType=SAMPLERTYPE_Masks
   texture=/BRCosmetics/Weapons/FORT_Melee/Pickaxe_HydroBottle/Textures/T_ScreenTilable
  }
  n13 = LinearInterpolate(A=@n12.R, B=Multiply(A=@n12.R, B=NamedRerouteUsage()), Alpha=NamedRerouteUsage())
  n14 = Add(
   A=NamedRerouteDeclaration(None=LinearInterpolate(A=@n1.RGB, B=@n2.RGB, Alpha=@n3.R)) {name=ScreenBackGround, nodeColor=<rgba 1 0.8438 0 1>}
   B=Multiply(A=VectorParameter() {group=Scifi_Screen, parameterName=TextYellow, defaultValue=<rgba 0.8388 0.5354 0.0994 1>}.RGB, B=Add(A=@n13, B=@n12.G))
  )
  Add(A=@n14, B=NamedRerouteUsage())
  Frac(None=@n11)
  n15 = MaterialFunctionCall() {materialFunction=&MF_CharacterParent_2023_PreFX}
  n16 = GetMaterialAttributes(None=@n15.Result) {attributeGetTypes [#1]}
  n17 = NamedRerouteUsage()
  n18 = NamedRerouteDeclaration(None=ChannelMaskParameter(None=@n15.Tech Art Mask) {group=Scifi_Screen, maskChannel=Green, parameterName=ScreenChannel, defaultValue=<rgba 0 1 0 0>}.R) {name=TechArtMask, nodeColor=<rgba 0 0.8359 1 1>}
  MaterialFunctionCall(
   Material=SetMaterialAttributes(MaterialAttributes=@n16.MaterialAttributes, Emissive Color=LinearInterpolate(A=@n16.Emissive Color, B=@n17, Alpha=@n18)) {attributeSetTypes [#1]}
   M=@n15.M
   Pre Skinned Local Position=@n15.M
   Distance Blend=@n15.M
   Fresnel=@n15.M
  ) {materialFunction=&MF_CharacterParent_2023_PostFX}
  Multiply(A=@n13, B=OneMinus(None=NamedRerouteUsage()))
  NamedRerouteDeclaration(None=@n4.ScaleY) {name=ScaleY, nodeColor=<rgba 0.2422 0 1 1>}
  n19 = NamedRerouteUsage()
  n20 = Constant() {r=1}
  NamedRerouteDeclaration(None=If(
    A=ComponentMask(None=@n19) {g=true}
    B=ScalarParameter() {group=Scifi_Screen, parameterName=BottomLine}
    "A > B"=Constant()
    "A == B"=@n20
    "A < B"=@n20
   )) {name=BottomMask, nodeColor=<rgba 0.4531 1 0 1>}
  NamedRerouteDeclaration(None=@n14) {name=NewEM, nodeColor=<rgba 0.6172 1 0 1>}
  n21 = NamedRerouteUsage()
  n22 = ComponentMask(None=@n21) {r=true}
  n23 = VectorParameter() {group=Scifi_Screen, parameterName=TextBoundingBox, channelNames {a=YMax, b=YMin, g=XMax, r=XMin}, defaultValue=<rgba 0 1 0 1>}
  n24 = Constant()
  n25 = If(A=@n22, B=@n23.XMin, "A > B"=Constant() {r=1}, "A == B"=@n24, "A < B"=@n24)
  n26 = Constant()
  n27 = OneMinus(None=If(A=@n22, B=NamedRerouteUsage(), "A > B"=@n26, "A == B"=@n26, "A < B"=Constant() {r=1}))
  n28 = ComponentMask(None=@n21) {g=true}
  n29 = Constant()
  n30 = Constant()
  n31 = Multiply(
   A=If(A=@n28, B=@n23.YMin, "A > B"=Constant() {r=1}, "A == B"=@n29, "A < B"=@n29)
   B=If(A=@n28, B=@n23.YMax, "A > B"=@n30, "A == B"=@n30, "A < B"=Constant() {r=1})
  )
  NamedRerouteDeclaration(None=Multiply(A=Multiply(A=@n25, B=@n27), B=@n31)) {name=AreaMaskDynamic, nodeColor=<rgba 1 0.2109 0 1>}
  NamedRerouteDeclaration(None=@n23.XMin) {name=XMin, nodeColor=<rgba 0.2891 0 1 1>}
  NamedRerouteDeclaration(None=@n23.XMax) {name=XMax, nodeColor=<rgba 0.2891 0 1 1>}
  NamedRerouteDeclaration(None=@n10) {name=TimeAdj, nodeColor=<rgba 1 0 0.3516 1>}
  n32 = ScalarParameter() {defaultValue=20, parameterName=HorizontalDivisor}
  n33 = Multiply(A=NamedRerouteUsage(), B=@n32)
  n34 = Divide(A=Fmod(A=Floor(None=@n33), B=@n32), B=@n32)
  NamedRerouteDeclaration(None=LinearInterpolate(A=NamedRerouteUsage(), B=NamedRerouteUsage(), Alpha=OneMinus(None=@n34))) {name=XMax_Line, nodeColor=<rgba 0 0.6719 1 1>}
  n35 = Constant()
  n36 = If(A=@n22, B=NamedRerouteUsage(), "A > B"=@n35, "A == B"=@n35, "A < B"=Constant() {r=1})
  NamedRerouteDeclaration(None=Multiply(A=Multiply(A=@n25, B=@n36), B=@n31)) {name=AreaMaskStatic, nodeColor=<rgba 0.7578 1 0 1>}
 ]
}
