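# Physics asset for the HydroBottle pickaxe, a Battle Royale cosmetic melee weapon mesh.
# Two capsule bodies cover the pickaxe bone and a smaller cable bone.
# A single constraint joins them with angular limits, loosest on the first swing axis and tight on twist, which restricts how far the two can rotate relative to each other.
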PhysicsAsset /BRCosmetics/Weapons/FORT_Melee/Pickaxe_HydroBottle/Meshes/HydroBottle_Axe_Physics
bodies {
 dyn_cable [{shapeType=Capsule, length=12, radius=2.25, shapeName=dyn_cable_capsule, center=<-1.1938 0 0>, rotation=<rot 0 0 180>}]
 dyn_hydrobottle_pickaxe [{shapeType=Capsule, length=26.3454, radius=5.943, center=<0.4154 0.5962 12.2161>, rotation=<rot 5.1065 -90 2.3139>}]
}
constraints [{bone1Name=dyn_hydrobottle_pickaxe, bone2Name=dyn_cable, swing1LimitDegrees=60, swing2LimitDegrees=15, twistLimitDegrees=5}]
